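; Weapon Blueprint for a Creative ranged flashlight pistol that carries a spot light.
; When equipped and attached, it plays the equip sound attached to the weapon mesh.
; It then enables shadow casting on the spot light only if the instigator pawn is locally controlled, so other clients see just the light.
; The construction script only calls the parent.
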(event FortWeapon|EventOnEquippedAndAttached
  ;[1 1 1] Only Local players (owning player) see the shadow from the spot light.  All other clients just see light
  (Audio|SpawnSoundAttached (Variables|Sounds|GetEquip%20Sound) (Variables|Visual|GetWeaponMesh))
  (if (Pawn|IsLocallyControlled (Game|GetInstigator))
    (Rendering|Components|Light|SetCastShadows (Variables|Default|GetSpotLight) true)))

(fn ConstructionScript
  (|CallParentFunction))
