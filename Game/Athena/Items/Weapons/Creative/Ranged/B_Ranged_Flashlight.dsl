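; Creative flashlight ranged weapon Blueprint that sets up its spotlight and light-beam mesh according to platform and quality settings.
; On BeginPlay, it plays an attached equip sound and enables shadow casting only for the locally controlled instigator off mobile.
; Low shadow quality or detail mode 1 drops the light function, uses a single-sided beam material and scales intensity from the construction-script cached base.
; All this logic runs only where cosmetic events can execute.
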(event EventBeginPlay
  (CallFunction|InitComponents)
  (CallFunction|InitLightShadowQuality))

(fn InitComponents
  (Networking|CanExecuteCosmeticEvents
    (:True
      (Audio|SpawnSoundAttached (Variables|Sounds|GetEquip%20Sound) (Variables|Visual|GetWeaponMesh))
      (Rendering|Components|Light|SetCastShadows (Variables|Default|GetSpotLight) (and (Pawn|IsLocallyControlled (Game|GetInstigator)) (not (Fortnite|IsMobilePlatform)))))))

(fn InitLightShadowQuality
  (bind _returnvalue (Rendering|GetRenderingDetailMode))
  (bind _returnvalue_2 (== _returnvalue 1))
  (Networking|CanExecuteCosmeticEvents
    (:True
      (bind _returnvalue_3 (Settings|GetGameUserSettings))
      (Rendering|Components|Light|SetLightFunctionMaterial (Variables|Default|GetSpotLight) (select (or (<= (Settings|GetShadowQuality _returnvalue_3) 1) _returnvalue_2) 0 (Variables|Material|GetLightFunctionMaterial)))
      (Rendering|Material|SetMaterial (Variables|Default|GetSm_CreativeMoveTool_Lightbeam) 0 (select (or (<= (Settings|GetShadowQuality _returnvalue_3) 1) _returnvalue_2) "/Game/Athena/Items/Weapons/Creative/Ranged/M_Flashlight_LightRay_SingleSided_Inst.M_Flashlight_LightRay_SingleSided_Inst" "/Game/Athena/Items/Weapons/Creative/Ranged/M_Flashlight_LightRay.M_Flashlight_LightRay"))
      (Rendering|Components|Light|SetIntensity (Variables|Default|GetSpotLight) (select (or (<= (Settings|GetShadowQuality _returnvalue_3) 1) _returnvalue_2) (* (Variables|Light|GetSpotlightBaseIntensity) (Variables|Light|GetSpotlightLowQualityIntensityScale)) (Variables|Light|GetSpotlightBaseIntensity))))))

(fn ConstructionScript
  (Networking|CanExecuteCosmeticEvents
    (:True
      (Variables|Light|SetSpotlightBaseIntensity (Variables|/Script/Engine.LightComponentBase|GetIntensity (Variables|Default|GetSpotLight))))))
