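# Translucent unlit deferred decal material that fakes a colored flashlight beam on surfaces, apparently for mobile, judging by the name.
# It measures each pixel's normalized distance along the decal's orientation axis from an offset origin.
# That distance widens a UV sphere mask into a cone and drives an inverse power falloff.
# A decal-normal term fades surfaces facing along the axis, and the product is tinted by Color.
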Material /Game/Athena/Items/Weapons/Creative/Ranged/MD_FlashlightMobile
graph {
 properties {
  blendMode=BLEND_Translucent
  materialDomain=MD_DeferredDecal
  pixelDepthOffsetMode=PDOM_Legacy
  shadingModel=MSM_Unlit
  usedShadingModels=Unlit
 }
 nodes [
  n1 = Multiply(A=ScalarParameter() {defaultValue=-720, parameterName=ZOffset}, B=Sign(None=ObjectOrientation()))
  n2 = ObjectOrientation()
  n3 = Subtract(
   A=WorldPosition().XYZ
   B=Add(
    A=ActorPositionWS()
    B=MaterialFunctionCall(VectorToTransform=@n1, BasisPosition=@n2) {materialFunction=&Transform3x3Matrix}.Transformed Vector
   )
  )
  n4 = ObjectOrientation()
  n5 = Saturate(None=Divide(A=DotProduct(A=@n3, B=@n4), B=ScalarParameter() {defaultValue=1440, parameterName=DecalSize}))
  n6 = Power(
   Base=SphereMask(A=TextureCoordinate(), B=Constant() {r=0.5}, Radius=LinearInterpolate(Alpha=@n5) {constB=0.5}, Hardness=Constant())
   Exp=ScalarParameter() {defaultValue=4, parameterName=LightConeExponent}
  )
  n7 = Power(Base=OneMinus(None=@n5), Exp=ScalarParameter() {defaultValue=1.5, parameterName=LightFalloffExponent})
  n8 = MaterialFunctionCall() {materialFunction=/Game/Weapons/Prototype/SprayGun/Materials/MF_DecalNormal}
  n9 = ObjectOrientation()
  n10 = OneMinus(None=Saturate(None=DotProduct(A=@n8.WorldNormal, B=@n9)))
  Multiply(
   A=VectorParameter() {parameterName=Color, defaultValue=<rgba 2 1.2847 0.2882 1>}.RGB
   B=Multiply(A=Multiply(A=@n6, B=@n7), B=Power(Base=@n10) {constExponent=3})
  )
 ]
}
