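# Unlit, additive, two-sided material for a flashlight light ray, tinted by a color parameter with a warm default.
# A double-sided Fresnel function shapes the glow, with brightness and exponent blended between top and bottom values by a power-curve gradient along UV V.
# That gradient, plus a biased squared-V term, scales the glow along the ray.
# Distance-behind-translucency and near-camera fades soften it near scene geometry and the camera.
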Material /Game/Athena/Items/Weapons/Creative/Ranged/M_Flashlight_LightRay
graph {
 properties {blendMode=BLEND_Additive, pixelDepthOffsetMode=PDOM_Legacy, shadingModel=MSM_Unlit, twoSided=true, usedShadingModels=Unlit}
 nodes [
  n1 = ScalarParameter() {defaultValue=1.5, parameterName=FresnelBrightnessTop}
  n2 = ScalarParameter() {defaultValue=2, parameterName=FresnelBrightnessBottom}
  n3 = ComponentMask(None=TextureCoordinate()) {g=true}
  n4 = Power(Base=@n3, Exp=ScalarParameter() {defaultValue=20, parameterName=GradientFalloffExponent})
  n5 = ScalarParameter() {defaultValue=4, parameterName=FresnelExponentTop}
  n6 = ScalarParameter() {defaultValue=0.5, parameterName=FresnelExponentBottom}
  n7 = Multiply(
   A=VectorParameter() {parameterName=Color, defaultValue=<rgba 5 3.2878 1.0156 0>}.RGB
   B=MaterialFunctionCall(Brightness=LinearInterpolate(A=@n1, B=@n2, Alpha=@n4), Exponent=LinearInterpolate(A=@n5, B=@n6, Alpha=@n4)) {materialFunction=&MF_Fresnel_FX}.Fresnel Double Sided
  )
  n8 = Add(A=@n4, B=Multiply(A=Multiply(A=@n3, B=@n3), B=ScalarParameter() {defaultValue=0.045, parameterName=Bias}))
  n9 = ScalarParameter() {defaultValue=80, parameterName="Distance Divisor (S)"}
  Multiply(
   A=Multiply(
    A=Multiply(A=@n7, B=@n8)
    B=MaterialFunctionCall!(Distance Bias, Distance Divisor=@n9, Scene Depth, Pixel Depth) {materialFunction=/Game/Effects/Fort_Effects/Materials/_Functions/MF_DistanceBehindTranslucency_01}.Normalized
   )
   B=MaterialFunctionCall(Distance=ScalarParameter() {defaultValue=100, parameterName=NearCameraFade}) {materialFunction=&Near_Camera_Fade}.Result
  )
 ]
}
