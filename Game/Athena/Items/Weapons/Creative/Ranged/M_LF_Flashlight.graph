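# Light function material for a flashlight weapon in Creative that shapes the beam into a circular pattern with a hot center and uneven falloff.
# A radial gradient, scaled and biased, serves as the UV into a cloud noise texture, turning the noise into concentric rings.
# Those rings are faded by the gradient and a harder outer mask, then lerped toward white by an exponent-sharpened inner mask.
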Material /Game/Athena/Items/Weapons/Creative/Ranged/M_LF_Flashlight
graph {
 properties {
  bForceCompatibleWithLightFunctionAtlas=true
  materialDomain=MD_LightFunction
  pixelDepthOffsetMode=PDOM_Legacy
  shadingModel=MSM_Unlit
  usedShadingModels=Unlit
 }
 nodes [
  n1 = TextureCoordinate()
  n2 = Constant() {r=0.5}
  n3 = SphereMask(A=@n1, B=@n2, Radius=@n2, Hardness=ScalarParameter() {defaultValue=0.9, parameterName=HardnessOuter})
  n4 = ScalarParameter() {defaultValue=0.8, parameterName=OuterBrightness}
  n5 = ScalarParameter() {defaultValue=0.6, parameterName=BiasUV}
  n6 = Constant() {r=0.5}
  n7 = SphereMask(A=TextureCoordinate(), B=@n6, Radius=@n6, Hardness=Constant())
  n8 = Multiply(A=ScalarParameter() {defaultValue=0.04, parameterName=ScaleUV}, B=@n7)
  n9 = Add(
   A=TextureSample(UVs=Add(A=@n5, B=@n8)) {texture=&AtmosphericCloudNoise01}.R
   B=ScalarParameter() {defaultValue=0.0125, parameterName=NoiseBias}
  )
  n10 = ScalarParameter() {defaultValue=0.15, parameterName=RadiusInner}
  n11 = ScalarParameter() {defaultValue=0.1, parameterName=HarnessInner}
  LinearInterpolate(
   A=Multiply(A=Multiply(A=@n3, B=@n4), B=Multiply(A=@n9, B=@n7))
   Alpha=Power(Base=SphereMask(A=@n1, B=@n2, Radius=@n10, Hardness=@n11), Exp=ScalarParameter() {defaultValue=4, parameterName=InnerExponent})
  )
 ]
}
